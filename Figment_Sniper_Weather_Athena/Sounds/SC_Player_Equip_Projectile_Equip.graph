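# Sound cue for a player equip foley sound in the Figment Sniper Weather content, named for projectile equipping.
# A Random node with equal weights picks one of two melee equip waves and is excluded from branch culling.
# It plays in the player foley sound class with the generic player equip attenuation and slightly reduced volume.
# A concurrency override prevents new instances once the limit is reached.
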SoundCue /Figment_Sniper_Weather_Athena/Sounds/SC_Player_Equip_Projectile_Equip
root=Random_2
nodes [
 {id=Random_2, children [WavePlayer_1,WavePlayer_2], properties {bShouldExcludeFromBranchCulling=true, weights [1,1]}}
 {id=WavePlayer_1, properties {soundWaveAssetPtr=&/player_melee_equip_02}}
 {id=WavePlayer_2, properties {soundWaveAssetPtr=&/player_melee_equip_03}}
]
properties {
 attenuationSettings=/Game/Sounds/Attenuation/Fort_Player_Equip_Generic_Att
 bOverrideConcurrency=true
 maxDistance=1800
 soundClassObject=&Fort_Player_Foley
 volumeMultiplier=0.8
 concurrencyOverrides {resolutionRule=PreventNew}
}
