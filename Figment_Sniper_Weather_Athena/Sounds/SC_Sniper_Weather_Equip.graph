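# Sound cue for equipping the Weather sniper rifle, played as weapon foley with generic player equip attenuation.
# A mixer layers a quiet random pick of two player melee equip sounds under a louder, briefly delayed random pick of three Weather sniper reload-end sounds.
# The reload layer is pitched slightly down, and the melee layer's random node is excluded from branch culling.
# The cue overrides concurrency and has debug enabled.
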SoundCue /Figment_Sniper_Weather_Athena/Sounds/SC_Sniper_Weather_Equip
root=Mixer_0
nodes [
 {id=Delay_0, children [Modulator_0], properties {delayMax=0.1, delayMin=0.1}}
 {id=Mixer_0, children [Delay_0,Random_2], properties {inputVolume [0.8,0.2]}}
 {id=Modulator_0, children [Random_0], properties {pitchMax=0.9, pitchMin=0.9, volumeMax=1, volumeMin=1}}
 {id=Random_0, children [WavePlayer_1,WavePlayer_2,WavePlayer_3], properties {weights [1,1,1]}}
 {id=Random_2, children [WavePlayer_0,WavePlayer_4], properties {bShouldExcludeFromBranchCulling=true, weights [1,1]}}
 {id=WavePlayer_0, properties {soundWaveAssetPtr=&/player_melee_equip_02}}
 {id=WavePlayer_1, properties {soundWaveAssetPtr=/Game/Sounds/Fort_Weapons/SniperRifles/Weather/Weather_Sniper_ReloadEnd_01}}
 {id=WavePlayer_2, properties {soundWaveAssetPtr=/Game/Sounds/Fort_Weapons/SniperRifles/Weather/Weather_Sniper_ReloadEnd_02}}
 {id=WavePlayer_3, properties {soundWaveAssetPtr=/Game/Sounds/Fort_Weapons/SniperRifles/Weather/Weather_Sniper_ReloadEnd_03}}
 {id=WavePlayer_4, properties {soundWaveAssetPtr=&/player_melee_equip_03}}
]
properties {
 attenuationSettings=/Game/Sounds/Attenuation/Fort_Player_Equip_Generic_Att
 bDebug=true
 bOverrideConcurrency=true
 maxDistance=1800
 soundClassObject=&Fort_WeaponFoley
 volumeMultiplier=1
}
