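# Sound cue for the end of the Weather sniper rifle's reload.
# A Random node picks with equal weight among three reload-end wave variations.
# It plays as weapon foley with the shared weapon reload attenuation, at slightly lowered pitch and volume, with concurrency overridden.
# The debug flag is left on.
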SoundCue /Figment_Sniper_Weather_Athena/Sounds/SC_Sniper_Weather_ReloadEnd
root=Random_0
nodes [
 {id=Random_0, children [WavePlayer_1,WavePlayer_2,WavePlayer_3], properties {weights [1,1,1]}}
 {id=WavePlayer_1, properties {soundWaveAssetPtr=/Game/Sounds/Fort_Weapons/SniperRifles/Weather/Weather_Sniper_ReloadEnd_01}}
 {id=WavePlayer_2, properties {soundWaveAssetPtr=/Game/Sounds/Fort_Weapons/SniperRifles/Weather/Weather_Sniper_ReloadEnd_02}}
 {id=WavePlayer_3, properties {soundWaveAssetPtr=/Game/Sounds/Fort_Weapons/SniperRifles/Weather/Weather_Sniper_ReloadEnd_03}}
]
properties {
 attenuationSettings=&Fort_Weapon_Reload_Att
 bDebug=true
 bOverrideConcurrency=true
 maxDistance=2250
 pitchMultiplier=0.9
 soundClassObject=&Fort_WeaponFoley
 volumeMultiplier=0.8
}
