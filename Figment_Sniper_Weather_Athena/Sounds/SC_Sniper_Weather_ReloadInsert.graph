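# Sound cue for the reload insert foley of the Weather sniper rifle.
# A random node picks between two insert wave variations with equal weight.
# The cue plays at reduced volume and slightly lowered pitch through the weapon foley sound class and shared weapon reload attenuation.
# It overrides concurrency settings and has debug enabled.
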SoundCue /Figment_Sniper_Weather_Athena/Sounds/SC_Sniper_Weather_ReloadInsert
root=Random_0
nodes [
 {id=Random_0, children [WavePlayer_1,WavePlayer_2], properties {weights [1,1]}}
 {id=WavePlayer_1, properties {soundWaveAssetPtr=/Game/Sounds/Fort_Weapons/SniperRifles/Weather/Weather_Sniper_ReloadInsert_01}}
 {id=WavePlayer_2, properties {soundWaveAssetPtr=/Game/Sounds/Fort_Weapons/SniperRifles/Weather/Weather_Sniper_ReloadInsert_02}}
]
properties {
 attenuationSettings=&Fort_Weapon_Reload_Att
 bDebug=true
 bOverrideConcurrency=true
 maxDistance=2250
 pitchMultiplier=0.9
 soundClassObject=&Fort_WeaponFoley
 volumeMultiplier=0.5
}
