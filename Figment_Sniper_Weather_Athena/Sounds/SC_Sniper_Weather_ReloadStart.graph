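# Sound cue for the start of the Weather sniper rifle reload in the Figment weapon set.
# A single wave player feeds one reload-start recording straight to the output, with no randomization or modulation nodes.
# The cue lowers volume and pitch slightly, uses the shared weapon reload attenuation and the weapon foley sound class, overrides concurrency, and has its debug flag on.
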SoundCue /Figment_Sniper_Weather_Athena/Sounds/SC_Sniper_Weather_ReloadStart
root=WavePlayer_1
nodes [{id=WavePlayer_1, properties {soundWaveAssetPtr=/Game/Sounds/Fort_Weapons/SniperRifles/Weather/Weather_Sniper_ReloadStart_01}}]
properties {
 attenuationSettings=&Fort_Weapon_Reload_Att
 bDebug=true
 bOverrideConcurrency=true
 maxDistance=2250
 pitchMultiplier=0.9
 soundClassObject=&Fort_WeaponFoley
 volumeMultiplier=0.8
}
