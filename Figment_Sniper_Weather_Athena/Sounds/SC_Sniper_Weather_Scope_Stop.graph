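# Sound cue for the stop sound of the weather sniper rifle's scope, classed as weapon foley.
# A single wave player feeds the output directly, with lowered pitch and reduced volume.
# The cue overrides attenuation to set its own falloff and non-spatialized radius, and overrides concurrency to prevent new instances rather than cut off playing ones.
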SoundCue /Figment_Sniper_Weather_Athena/Sounds/SC_Sniper_Weather_Scope_Stop
root=WavePlayer_0
nodes [{id=WavePlayer_0, properties {soundWaveAssetPtr=/Game/Sounds/Fort_Weapons/SniperRifles/Weather/Weather_Sniper_Scope_Stop_01}}]
properties {
 bOverrideAttenuation=true
 bOverrideConcurrency=true
 maxDistance=2700
 pitchMultiplier=0.8
 soundClassObject=&Fort_WeaponFoley
 volumeMultiplier=0.23
 attenuationOverrides {falloffDistance=1500, nonSpatializedRadiusStart=200, attenuationShapeExtents {x=1200}}
 concurrencyOverrides {resolutionRule=PreventNew}
}
