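# SC_UI_DropWeaponImpact is a Sound Cue for the impact of a dropped weapon in the Figment Sniper Weather content, judging by its name.
# A single wave player feeds the UI_DropWeaponImpact01 wave straight to the output, with no randomization or mixing.
# The cue plays at reduced volume, uses weapon-pickup attenuation and is assigned to the distant ranged-weapon fire sound class.
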SoundCue /Figment_Sniper_Weather_Athena/Sounds/SC_UI_DropWeaponImpact
root=WavePlayer_0
nodes [{id=WavePlayer_0, properties {soundWaveAssetPtr=&/UI_DropWeaponImpact01}}]
properties {
 attenuationSettings=&Fort_Weapon_Pickup_Att
 maxDistance=3500
 soundClassObject=&Fort_WeaponRanged_Fire_Distant
 volumeMultiplier=0.4
}
